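; Cosmetic pickaxe Blueprint for TwilightSpotSpell that switches the visible material element on the weapon as hits land.
; Weak-point hits, plus impacts in preview or frontend, call SetStageByIndex.
; Outside preview and frontend a counter makes every third weak-point hit advance the stage, while preview advances on each call and wraps to zero.
; Each of the four stages shows one element by setting "Hide Element" scalar parameters on the weapon mesh materials.
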(event EventBeginPlay
  (Variables|Default|SetStage 0))

(event EventTick
  (params (DeltaSeconds :type Float%20%28single-precision%29)))

(event FortWeapon|EventOnPlayImpactFX
  (params (HitResult :type Hit%20Result%20Structure%20%28by%20ref%29) (ImpactPhysicalSurface :type EPhysicalSurface%20Enum :default "SurfaceType_Default") (SpawnedPSC :type FXSystem%20Component%20Object%20Reference))
  (if (or (Reactive|Frontend|IsCosmeticObjectDisplayedasPreview 0) (Reactive|Frontend|IsCosmeticObjectDisplayedinFrontendContext 0))
    (CallFunction|SetStagebyIndex)))

(event EventOnPlayWeakpointHitFX
  (CallFunction|SetStagebyIndex))

(event Custom|SetStageByIndex
  (bind _returnvalue (Reactive|Frontend|IsCosmeticObjectDisplayedasPreview 0))
  (bind _returnvalue_2 (Reactive|Frontend|IsCosmeticObjectDisplayedinFrontendContext 0))
  (bind _returnvalue_3 (or _returnvalue _returnvalue_2))
  (if (or _returnvalue_3 (>= (Variables|Default|GetWeakPointHit) 2))
    (Variables|Default|SetWeakPointHit 0)
    (bind _result (Math|Integer|IncrementInt (Variables|Default|GetStage))
      (:%20%20
        (if (>= _result 4)
          (if _returnvalue_3
            (Variables|Default|SetStage 0))
          (else
            _))
        (switch int (Variables|Default|GetStage)
          (:0
            (Rendering|Material|SetScalarParameterValueonMaterials (Variables|Visual|GetWeaponMesh) "Hide Element 01" 0.0)
            (Rendering|Material|SetScalarParameterValueonMaterials (Variables|Visual|GetWeaponMesh) "Hide Element 02" 1.0)
            (Rendering|Material|SetScalarParameterValueonMaterials (Variables|Visual|GetWeaponMesh) "Hide Element 03" 1.0)
            (Rendering|Material|SetScalarParameterValueonMaterials (Variables|Visual|GetWeaponMesh) "Hide Element 04" 1.0))
          (:1
            (Rendering|Material|SetScalarParameterValueonMaterials (Variables|Visual|GetWeaponMesh) "Hide Element 01" 1.0)
            (Rendering|Material|SetScalarParameterValueonMaterials (Variables|Visual|GetWeaponMesh) "Hide Element 02" 0.0)
            (Rendering|Material|SetScalarParameterValueonMaterials (Variables|Visual|GetWeaponMesh) "Hide Element 03" 1.0)
            (Rendering|Material|SetScalarParameterValueonMaterials (Variables|Visual|GetWeaponMesh) "Hide Element 04" 1.0))
          (:2
            (Rendering|Material|SetScalarParameterValueonMaterials (Variables|Visual|GetWeaponMesh) "Hide Element 01" 1.0)
            (Rendering|Material|SetScalarParameterValueonMaterials (Variables|Visual|GetWeaponMesh) "Hide Element 02" 1.0)
            (Rendering|Material|SetScalarParameterValueonMaterials (Variables|Visual|GetWeaponMesh) "Hide Element 03" 0.0)
            (Rendering|Material|SetScalarParameterValueonMaterials (Variables|Visual|GetWeaponMesh) "Hide Element 04" 1.0))
          (:3
            (Rendering|Material|SetScalarParameterValueonMaterials (Variables|Visual|GetWeaponMesh) "Hide Element 01" 1.0)
            (Rendering|Material|SetScalarParameterValueonMaterials (Variables|Visual|GetWeaponMesh) "Hide Element 02" 1.0)
            (Rendering|Material|SetScalarParameterValueonMaterials (Variables|Visual|GetWeaponMesh) "Hide Element 03" 1.0)
            (Rendering|Material|SetScalarParameterValueonMaterials (Variables|Visual|GetWeaponMesh) "Hide Element 04" 0.0)))))
    (else
      (Math|Integer|IncrementInt (Variables|Default|GetWeakPointHit)))))

(fn ConstructionScript
  (|CallParentFunction))
